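# Audio Level Sequence for the moc1000 sequence of PistolCanary, apparently the "Moment of Calm" going by its ambience track name.
# It possesses the PistolCanary_AudioManager actor, firing an event through a trigger track and setting its Music Phase integer.
# A short MetaSound5.1 audio section at negative frames exposes listener azimuth and master volume as channels.
# The MomentOfCalmAmbience track has no sections.
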LevelSequence /PistolCanary/Sequence/moc/moc1000/PistolCanary_moc1000_aud
bindings [PistolCanary_AudioManager {
  id=#1
  possessed_class=PistolCanary_AudioManager_C
  tracks [
   {
    type=MovieSceneEventTrack
    sections [{
      type=MovieSceneEventTriggerSection
      channels [None {type=MovieSceneScriptingEventChannel, const ["<Struct 'MovieSceneEvent' (0x852ff017800) {}>", 1]}]
     }]
   }
   {
    type=MovieSceneIntegerTrack
    display_name=Music Phase
    sections [{type=MovieSceneIntegerSection, channels [None {type=MovieSceneScriptingIntegerChannel, default=11, const [11,1]}]}]
   }
  ]
 }]
display_rate {denominator=1, numerator=60}
tick_resolution {denominator=1, numerator=24000}
tracks [
 {
  type=MovieSceneAudioTrack
  display_name=MetaSound5.1
  sections [{
    type=MovieSceneAudioSection
    end_frame=-15223
    start_frame=-15283
    channels [
     Volume {type=MovieSceneScriptingFloatChannel, default=1}
     Pitch {type=MovieSceneScriptingFloatChannel, default=1}
     UE.Listener.Orientation.AzimuthFMovieSceneFloatChannel {type=MovieSceneScriptingFloatChannel, default=117.031}
     MasterVolumeFMovieSceneFloatChannel {type=MovieSceneScriptingFloatChannel, default=1}
     AttachActor {type=MovieSceneScriptingActorReferenceChannel, default="<Struct 'MovieSceneObjectBindingID' (0x852fd7a5f60) {}>"}
    ]
   }]
 }
 {type=MovieSceneAudioTrack, display_name=MomentOfCalmAmbience, sections []}
]
