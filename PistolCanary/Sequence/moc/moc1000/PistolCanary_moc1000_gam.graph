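# Gameplay Level Sequence for moc1000 in PistolCanary, driving a tentacle boss, a castle knockback volume and timed NPC events.
# A visibility track holds the possessed TentacleBoss hidden.
# A transform track keeps the knockback volume at a fixed location and keys its uniform scale from very large down to small over a brief window.
# An NPCs event track fires two triggers, one just after that scale change begins and one later.
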LevelSequence /PistolCanary/Sequence/moc/moc1000/PistolCanary_moc1000_gam
bindings [
 TentacleBoss {
  id=#1
  possessed_class=PistolCanary_TentacleBoss_C
  tracks [{
    type=MovieSceneVisibilityTrack
    sections [{type=MovieSceneVisibilitySection, channels [None {type=MovieSceneScriptingBoolChannel, default=false, const [false,3]}]}]
   }]
 }
 BP_PistolCanary_KnockbackVolume_Castle {
  id=#2
  possessed_class=BP_PistolCanary_KnockbackVolume_C
  tracks [{
    type=MovieScene3DTransformTrack
    sections [{
      type=MovieScene3DTransformSection
      channels [
       Location.X {type=MovieSceneScriptingDoubleChannel, default=14424}
       Location.Y {type=MovieSceneScriptingDoubleChannel, default=70100}
       Location.Z {type=MovieSceneScriptingDoubleChannel, default=3428}
       Rotation.X {type=MovieSceneScriptingDoubleChannel, default=0}
       Rotation.Y {type=MovieSceneScriptingDoubleChannel, default=0}
       Rotation.Z {type=MovieSceneScriptingDoubleChannel, default=0}
       Scale.X {type=MovieSceneScriptingDoubleChannel, default=0.5, keys [[3600,137.5],[3628,0.5]]}
       Scale.Y {type=MovieSceneScriptingDoubleChannel, default=0.5, keys [[3600,137.5],[3628,0.5]]}
       Scale.Z {type=MovieSceneScriptingDoubleChannel, default=0.5, keys [[3600,137.5],[3628,0.5]]}
      ]
     }]
   }]
 }
]
display_rate {denominator=1, numerator=60}
tick_resolution {denominator=1, numerator=24000}
tracks [{
  type=MovieSceneEventTrack
  display_name=NPCs
  sections [{
    type=MovieSceneEventTriggerSection
    channels [None {
      type=MovieSceneScriptingEventChannel
      keys [[3602, "<Struct 'MovieSceneEvent' (0x852ff016d60) {}>"], [4450, "<Struct 'MovieSceneEvent' (0x852ff016e00) {}>"]]
     }]
   }]
 }]
