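# Lighting Level Sequence for shot moc1000 of PistolCanary, which fades out a spotlight and a day sequence override.
# Float tracks ramp the DaySequenceModifier's User Blend Weight from one to zero and the SpotLight intensity down to zero, starting together, with the blend finishing first.
# Constant visibility tracks hide three jungle fog actors and the EndLightBeam mesh, which also has a fixed, vertically stretched transform.
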LevelSequence /PistolCanary/Sequence/moc/moc1000/PistolCanary_moc1000_lgt
bindings [
 DaySequenceModifier {
  id=#1
  parent_id=#2
  possessed_class=DaySequenceModifierComponent
  tracks [{
    type=MovieSceneFloatTrack
    display_name=User Blend Weight
    sections [{type=MovieSceneFloatSection, channels [None {type=MovieSceneScriptingFloatChannel, default=0, keys [[3600,1],[3845,0]]}]}]
   }]
 }
 LightComponent0 {
  id=#3
  parent_id=#4
  possessed_class=SpotLightComponent
  tracks [{
    type=MovieSceneFloatTrack
    display_name=Intensity
    sections [{type=MovieSceneFloatSection, channels [None {type=MovieSceneScriptingFloatChannel, default=120000, keys [[3600,120000],[3940,0]]}]}]
   }]
 }
 BP_JungleFog_LowHanging {
  id=#5
  possessed_class=BP_JungleFog_C
  tracks [{
    type=MovieSceneVisibilityTrack
    sections [{type=MovieSceneVisibilitySection, channels [None {type=MovieSceneScriptingBoolChannel, default=false}]}]
   }]
 }
 RectLight_pistolCanary {id=#4, possessed_class=SpotLight, tracks []}
 BP_JungleFog {
  id=#6
  possessed_class=BP_JungleFog_C
  tracks [{
    type=MovieSceneVisibilityTrack
    sections [{type=MovieSceneVisibilitySection, channels [None {type=MovieSceneScriptingBoolChannel, default=false}]}]
   }]
 }
 BP_PistolCanary_EndLightBeam {
  id=#7
  possessed_class=FortStaticMeshActor
  tracks [
   {
    type=MovieScene3DTransformTrack
    sections [{
      type=MovieScene3DTransformSection
      channels [
       Location.X {type=MovieSceneScriptingDoubleChannel, default=14505}
       Location.Y {type=MovieSceneScriptingDoubleChannel, default=70207}
       Location.Z {type=MovieSceneScriptingDoubleChannel, default=29879}
       Rotation.X {type=MovieSceneScriptingDoubleChannel, default=0}
       Rotation.Y {type=MovieSceneScriptingDoubleChannel, default=0}
       Rotation.Z {type=MovieSceneScriptingDoubleChannel, default=0}
       Scale.X {type=MovieSceneScriptingDoubleChannel, default=5}
       Scale.Y {type=MovieSceneScriptingDoubleChannel, default=5}
       Scale.Z {type=MovieSceneScriptingDoubleChannel, default=200}
      ]
     }]
   }
   {
    type=MovieSceneVisibilityTrack
    sections [{type=MovieSceneVisibilitySection, channels [None {type=MovieSceneScriptingBoolChannel, default=false}]}]
   }
  ]
 }
 DaySequenceModifierVolume_PistolCanary {id=#2, possessed_class=DaySequenceModifierVolume, tracks []}
 BP_JungleFog_LowHanging2 {
  id=#8
  possessed_class=BP_JungleFog_C
  tracks [{
    type=MovieSceneVisibilityTrack
    sections [{type=MovieSceneVisibilitySection, channels [None {type=MovieSceneScriptingBoolChannel, default=false}]}]
   }]
 }
]
display_rate {denominator=1, numerator=60}
tick_resolution {denominator=1, numerator=24000}
tracks []
