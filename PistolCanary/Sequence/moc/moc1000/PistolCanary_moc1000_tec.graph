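# Level sequence for PistolCanary moc1000 (tec) that drives environment and material changes, apparently for a scripted moment around the MaskMakerFortress area.
# An event track triggers an event on the PanelManager, while the fortress DaySequenceModifier's User Blend Weight is held at full.
# A material parameter collection track holds GlobalOpacity near one, then fades it to zero.
# As that fade ends, BP_SpiralClouds begins a short, steep rise along Z.
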LevelSequence /PistolCanary/Sequence/moc/moc1000/PistolCanary_moc1000_tec
bindings [
 PistolCanary_PanelManager {
  id=#1
  possessed_class=PistolCanary_PanelManager_C
  tracks [{
    type=MovieSceneEventTrack
    sections [{
      type=MovieSceneEventTriggerSection
      channels [None {type=MovieSceneScriptingEventChannel, const ["<Struct 'MovieSceneEvent' (0x852ff014100) {}>", 1]}]
     }]
   }]
 }
 DSV_MaskMakerFortress {id=#2, possessed_class=DaySequenceModifierVolume, tracks []}
 DaySequenceModifier {
  id=#3
  parent_id=#2
  possessed_class=DaySequenceModifierComponent
  tracks [{
    type=MovieSceneFloatTrack
    display_name=User Blend Weight
    sections [{type=MovieSceneFloatSection, channels [None {type=MovieSceneScriptingFloatChannel, default=0, const [1,1]}]}]
   }]
 }
 BP_SpiralClouds {
  id=#4
  possessed_class=BP_SpiralClouds_C
  tracks [{
    type=MovieScene3DTransformTrack
    sections [{
      type=MovieScene3DTransformSection
      channels [
       Location.X {type=MovieSceneScriptingDoubleChannel, default=14198.0736}
       Location.Y {type=MovieSceneScriptingDoubleChannel, default=70555.3582}
       Location.Z {type=MovieSceneScriptingDoubleChannel, default=29321, keys [[3815,11948],[3880,29321]]}
       Rotation.X {type=MovieSceneScriptingDoubleChannel, default=0}
       Rotation.Y {type=MovieSceneScriptingDoubleChannel, default=0}
       Rotation.Z {type=MovieSceneScriptingDoubleChannel, default=0}
       Scale.X {type=MovieSceneScriptingDoubleChannel, default=1}
       Scale.Y {type=MovieSceneScriptingDoubleChannel, default=1}
       Scale.Z {type=MovieSceneScriptingDoubleChannel, default=1}
      ]
     }]
   }]
 }
]
display_rate {denominator=1, numerator=60}
tick_resolution {denominator=1, numerator=24000}
tracks [{
  type=MovieSceneMaterialParameterCollectionTrack
  display_name=PistolCanary_MPC
  sections [{
    type=MovieSceneParameterSection
    channels [GlobalOpacity {type=MovieSceneScriptingFloatChannel, default=0, thinned_from=4, keys [[-15040,1.0242],[3600,1.0242],[3815,0]]}]
   }]
 }]
