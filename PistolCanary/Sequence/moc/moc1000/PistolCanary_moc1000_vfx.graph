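# VFX Level Sequence for shot moc1000 of PistolCanary that plays the closing of portal effects over one short window.
# Transform keys shrink the tentacle ground portal and the rift gate despawn Niagara actors to zero scale.
# The rift gate actor also has a keyed spawn track, with its child component driven by a Niagara system life-cycle section and a deactivate key.
# A lightning spiral mesh is shown at the start and hidden partway through.
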LevelSequence /PistolCanary/Sequence/moc/moc1000/PistolCanary_moc1000_vfx
bindings [
 NS_PistolCanary_Tentacle_GroundPortal_01 {
  id=#1
  possessed_class=NiagaraActor
  tracks [{
    type=MovieScene3DTransformTrack
    sections [{
      type=MovieScene3DTransformSection
      channels [
       Location.X {type=MovieSceneScriptingDoubleChannel, default=14348}
       Location.Y {type=MovieSceneScriptingDoubleChannel, default=69801}
       Location.Z {type=MovieSceneScriptingDoubleChannel, default=3875, const [4073,1]}
       Rotation.X {type=MovieSceneScriptingDoubleChannel, default=0}
       Rotation.Y {type=MovieSceneScriptingDoubleChannel, default=0}
       Rotation.Z {type=MovieSceneScriptingDoubleChannel, default=0}
       Scale.X {type=MovieSceneScriptingDoubleChannel, default=0, keys [[3600,1.2],[3968,0]]}
       Scale.Y {type=MovieSceneScriptingDoubleChannel, default=0, keys [[3600,1.2],[3968,0]]}
       Scale.Z {type=MovieSceneScriptingDoubleChannel, default=0, keys [[3600,1.2],[3968,0]]}
      ]
     }]
   }]
 }
 NiagaraComponent0 {
  id=#2
  parent_id=#3
  possessed_class=NiagaraComponent
  tracks [
   {
    type=MovieSceneParticleTrack
    display_name=FX System
    sections [{type=MovieSceneParticleSection, channels [None {type=MovieSceneScriptingParticleChannel, const ["<ParticleKey.DEACTIVATE: 1>", 1]}]}]
   }
   {type=MovieSceneNiagaraSystemTrack, sections [{type=MovieSceneNiagaraSystemSpawnSection, end_frame=3968, start_frame=3600}]}
  ]
 }
 NS_SourFork_RiftGate_Portal_Despawn2 {
  id=#3
  possessed_class=NiagaraActor
  tracks [
   {
    type=MovieSceneSpawnTrack
    sections [{
      type=MovieSceneSpawnSection
      end_frame=14607
      channels [None {type=MovieSceneScriptingBoolChannel, default=false, keys [[3600,false],[3601,true],[3968,false]]}]
     }]
   }
   {
    type=MovieScene3DTransformTrack
    sections [{
      type=MovieScene3DTransformSection
      channels [
       Location.X {type=MovieSceneScriptingDoubleChannel, default=14564, const [14328,1]}
       Location.Y {type=MovieSceneScriptingDoubleChannel, default=70128, const [70128,1]}
       Location.Z {type=MovieSceneScriptingDoubleChannel, default=3896, const [3896,1]}
       Rotation.X {type=MovieSceneScriptingDoubleChannel, default=0, const [0,1]}
       Rotation.Y {type=MovieSceneScriptingDoubleChannel, default=0, const [0,1]}
       Rotation.Z {type=MovieSceneScriptingDoubleChannel, default=0, const [0,1]}
       Scale.X {type=MovieSceneScriptingDoubleChannel, default=0, keys [[3606,0.79],[3968,0]]}
       Scale.Y {type=MovieSceneScriptingDoubleChannel, default=0, keys [[3606,0.79],[3968,0]]}
       Scale.Z {type=MovieSceneScriptingDoubleChannel, default=0, keys [[3606,1],[3968,0]]}
      ]
     }]
   }
  ]
 }
 SM_MaskMaker_Lightning_3_Spirals_Segments2 {
  id=#4
  possessed_class=FortStaticMeshActor
  tracks [{
    type=MovieSceneVisibilityTrack
    sections [{
      type=MovieSceneVisibilitySection
      channels [None {type=MovieSceneScriptingBoolChannel, default=false, keys [[3600,true],[3812,false],[3968,false]]}]
     }]
   }]
 }
]
display_rate {denominator=1, numerator=60}
tick_resolution {denominator=1, numerator=24000}
tracks []
